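# Sound cue for the fade-in music of the Grasshopper emote in Battle Royale, routed to the emote music sound class with emote attenuation.
# A mixer layers two fade-up waves at equal volume, an IP version and a Float version.
# An enveloper on the IP layer holds near full volume for about two seconds, then fades it to near silence, leaving the Float layer.
# An emote foley concurrency set is assigned.
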SoundCue /BRCosmetics/Sounds/Emotes/Grasshopper/Emote_Grasshopper_FadeIn
root=Mixer_0
nodes [
 {
  id=Enveloper_0
  children [WavePlayer_1]
  properties {pitchCurve {editorCurveData {keys [{time=0}]}}, volumeCurve {editorCurveData {keys [<at 0 1>,<at 2.0137 0.9967>,<at 3.7245 0.0033>]}}}
 }
 {id=Mixer_0, children [Enveloper_0,WavePlayer_2], properties {inputVolume [1,1]}}
 {id=WavePlayer_1, properties {soundWaveAssetPtr=&/Emote_Grasshopper_Loop_IP_FadeUp_01}}
 {id=WavePlayer_2, properties {soundWaveAssetPtr=&/Emote_Grasshopper_Loop_Float_FadeUp_01}}
]
properties {
 attenuationSettings=&Fort_Emotes_Athena_Att
 maxDistance=1600
 soundClassObject=&Fort_Music_Emote
 volumeMultiplier=1
 concurrencyOverrides {maxCount=3, resolutionRule=StopOldest}
 concurrencySet [/Game/Sounds/Concurrency_Rules/Athena_EmoteFoley_Conc]
}
